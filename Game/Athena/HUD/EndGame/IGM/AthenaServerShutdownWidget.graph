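# Fortnite Athena HUD widget that shows a server shutdown countdown announcing the return to the lobby, built on AthenaShutdownWidget.
# It starts collapsed, and the parent's start and stop events toggle its visibility.
# Start and update events pass remaining time to SetCountdownText, which rounds it into the StartingCountdownText format string for the single CommonTextBlock.
# PreConstruct fills in a sample countdown for designer preview.
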WidgetBlueprint /Game/Athena/HUD/EndGame/IGM/AthenaServerShutdownWidget
graphs [
 EventGraph {nodes [
   n1 = Event PreConstruct()
   Branch(execute=[@n4.then, @n1.then], Condition=@n1.IsDesignTime)
   > SetCountdownText!(execute, self, RemainingTime:double=30) -> (then)
   Event Construct()
   > SetVisibility(InVisibility=Collapsed)
   n2 = Event OnUpdateCountdown!() -> (OutputDelegate, then, RemainingTime:float) {is=title K2Node_Event}
   > SetCountdownText(RemainingTime=@n2.RemainingTime)
   n3 = Event OnStartShowing!() -> (OutputDelegate, then, RemainingTime:float) {is=title K2Node_Event}
   > SetVisibility(InVisibility=Visible)
   > SetCountdownText(RemainingTime=@n3.RemainingTime)
   Event OnStopShowing!() -> (OutputDelegate, then, RemainingTime:float) {is=title K2Node_Event}
   > SetVisibility(InVisibility=Collapsed)
   n4 = Set StartingCountdownText(StartingCountdownText:Text="GetText (Text)"!(self:Text Object Reference=Get Text_Countdown() -> (Text_Countdown:Common Text Object Reference)) -> (ReturnValue:Text)) -> (Output_Get:Text)
  ]}
 SetCountdownText {nodes [
   n1 = SetCountdownText!() -> (then, RemainingTime:double=~) {is=title K2Node_FunctionEntry}
   > "SetText (Text)"(
    self=Get Text_Countdown()
    InText=Format Text!(Format=Get StartingCountdownText() -> (StartingCountdownText:Text), NUMBEROFSECONDS:int=Round(A=@n1) {is=title K2Node_CallFunction}) -> (Result)
   )
  ]}
]
variables [StartingCountdownText {type=string, default="RETURNING TO THE LOBBY IN {NUMBEROFSECONDS} SECONDS..."}]
widgets {
 info {inheritedWidgetCount=0, parentClass=/Script/FortniteUI.AthenaShutdownWidget, rootWidgetClass=&CommonTextBlock, widgetCount=1}
 widgets [Text_Countdown {
   widgetClassPath=&CommonTextBlock
   bIsVariable=true
   widget=Text_Countdown
   properties {
    text="Returning to the lobby IN {0} SECONDS..."
    textTransformPolicy=ToUpper
    font {
     bForceMonospaced=false
     bMaterialIsStencil=false
     fontMaterial=null
     fontObject=&Burbank
     letterSpacing=0
     monospacedWidth=1
     size=24
     skewAmount=0
     typefaceFontName=Burbank-PlayerPower
     outlineSettings {
      bApplyOutlineToDropShadows=false
      bMiteredCorners=false
      bSeparateFillAlpha=false
      outlineMaterial=null
      outlineSize=0
      outlineColor=<rgba 0 0 0 1>
     }
    }
    renderTransform {angle=0, scale=<1 1>, shear=<-8 0>, translation=<0 0>}
    shadowOffset=<0 0>
   }
  }]
}
