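# Party member widget for the Athena in-game menu, showing avatar, name, voice status, party-leader icon and readiness, scaled down when notPlayer is set.
# Parent-class events toggle widget and leader-icon visibility and set the readiness text.
# IsPlayerReady guards transitions: becoming Ready plays the Ready animation and a sound, leaving it reverses the animation at double speed.
# Extension widgets for the TeammateStatusPortraitOverlay_IGM slot fill an overlay on the avatar.
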WidgetBlueprint /Game/Athena/HUD/EndGame/IGM/InGameSocialUserWidget
animations [
 Selected {
  binding_count=1
  length_seconds=0.279
  object=&self:Selected
  sequence {bindings [Image_Selected_Effect {
     id=#1
     possessed_class=Image
     tracks [
      {
       type=MovieSceneFloatTrack
       display_name=Render Opacity
       sections [{type=MovieSceneFloatSection, channels [None {type=MovieSceneScriptingFloatChannel, keys [[0,1],[5.5803,0]]}]}]
      }
      {
       type=MovieScene2DTransformTrack
       sections [{
         type=MovieScene2DTransformSection
         channels [
          Scale.X {type=MovieSceneScriptingFloatChannel, keys [[0,2],[5.5803,0.9451]]}
          Scale.Y {type=MovieSceneScriptingFloatChannel, keys [[0,2],[5.5803,0.9451]]}
         ]
        }]
      }
     ]
    }]}
 }
 Ready {
  binding_count=4
  length_seconds=0.3
  object=&self:Ready
  sequence {bindings [
    ReadyCheckBox {id=#2, possessed_class=Image, tracks []}
    SizeBoxReadyText {
     id=#3
     possessed_class=SizeBox
     tracks [
      {
       type=MovieScene2DTransformTrack
       sections [{
         type=MovieScene2DTransformSection
         channels [
          Translation.X {type=MovieSceneScriptingFloatChannel}
          Translation.Y {type=MovieSceneScriptingFloatChannel, keys [[0,137.765],[2.7063,1.4096],[4.8947,-12.0113],[6,0]]}
         ]
        }]
      }
      {
       type=MovieSceneFloatTrack
       display_name=Render Opacity
       sections [{type=MovieSceneFloatSection, channels [None {type=MovieSceneScriptingFloatChannel, keys [[0,0],[3.3857,1]]}]}]
      }
     ]
    }
    CommonBorder_0 {
     id=#4
     possessed_class=CommonBorder
     tracks [{
       type=MovieSceneWidgetMaterialTrack
       sections [{
         type=MovieSceneParameterSection
         channels [
          BGColorR {type=MovieSceneScriptingFloatChannel, keys [[0,0.3419],[6,0.0908]]}
          BGColorG {type=MovieSceneScriptingFloatChannel, keys [[0,0],[6,0.5333]]}
          BGColorB {type=MovieSceneScriptingFloatChannel, keys [[0,0.5647],[6,1]]}
          BGColorA {type=MovieSceneScriptingFloatChannel, const [1,2]}
         ]
        }]
      }]
    }
    Overlay_1 {
     id=#5
     possessed_class=Overlay
     tracks [{
       type=MovieScene2DTransformTrack
       sections [{
         type=MovieScene2DTransformSection
         channels [
          Translation.X {type=MovieSceneScriptingFloatChannel}
          Translation.Y {type=MovieSceneScriptingFloatChannel}
          Angle {type=MovieSceneScriptingFloatChannel}
          Scale.X {type=MovieSceneScriptingFloatChannel, keys [[0,0],[4.055,1.2117],[4.2987,1.0488],[6,0.8438]]}
          Scale.Y {type=MovieSceneScriptingFloatChannel, keys [[0,0],[4.055,1.2117],[4.2987,1.0488],[6,0.8438]]}
          Shear.X {type=MovieSceneScriptingFloatChannel}
          Shear.Y {type=MovieSceneScriptingFloatChannel}
         ]
        }]
      }]
    }
   ]}
 }
]
graphs [
 BP_HandleExtensionWidgetCreated {nodes [
   n1 = Handle Extension Widget Created!() -> (then, Widget:User Widget Object Reference, SlotType:EUIExtensionSlot Enum=Invalid) {is=title K2Node_FunctionEntry}
   > Branch(Condition="Equal (Enum)"(A:EUIExtensionSlot Enum=@n1.SlotType, B:EUIExtensionSlot Enum=TeammateStatusPortraitOverlay_IGM))
   > n2 = AddChildToOverlay!(
    execute
    self:Overlay Object Reference=Get Overlay_AvatarExtension() -> (Overlay_AvatarExtension:Overlay Object Reference)
    Content:Widget Object Reference=Get Widget() -> (Widget:User Widget Object Reference)
   ) -> (then, ReturnValue:Overlay Slot Object Reference)
   > SetVerticalAlignment(self=@n2, InVerticalAlignment=VAlign_Fill)
   > SetHorizontalAlignment(self=@n2, InHorizontalAlignment=HAlign_Fill)
  ]}
 EventGraph {nodes [
   Event PreConstruct()
   > SetUserSpecifiedScale!(
    execute
    self:Scale Box Object Reference=Get ScaleBox_ItemSize() -> (ScaleBox_ItemSize:Scale Box Object Reference)
    InUserSpecifiedScale:float=Select(Option 0:double=1, Option 1:double=0.75, Index=Get notPlayer() -> (notPlayer:bool)) -> (ReturnValue:double)
   ) -> (then)
   n1 = Event OnIsPartyLeaderChanged!() -> (OutputDelegate, then, bIsPartyLeader:bool) {is=title K2Node_Event}
   > SetVisibility(
    self=Get Overlay_PartyLeaderIcon() -> (Overlay_PartyLeaderIcon:Overlay Object Reference)
    InVisibility=Select(Option 0:ESlateVisibility Enum=Collapsed, Option 1:ESlateVisibility Enum=SelfHitTestInvisible, Index=@n1.bIsPartyLeader) -> (ReturnValue:ESlateVisibility Enum)
   )
   Event OnUserSet() {is=title K2Node_Event}
   > SetVisibility(InVisibility=HitTestInvisible)
   Event OnWidgetReset() {is=title K2Node_Event}
   > SetVisibility(InVisibility=Collapsed)
   > SetVisibility(self=Get Overlay_PartyLeaderIcon(), InVisibility=Collapsed)
   n2 = Event OnNewReadiness!() -> (OutputDelegate, then, GameReadiness:EGameReadiness Enum=NotReady) {is=title K2Node_Event}
   > "SetText (Text)"(
    self=Get Text_Readiness() -> (Text_Readiness:Common Text Object Reference)
    InText=Select!(
     NotReady:Text="LOCTEXT(\"Not Ready\")"
     Ready:Text="LOCTEXT(\"Ready\")"
     Reevaluating:Text
     SittingOut:Text="LOCTEXT(\"Sitting Out\")"
     Away:Text="LOCTEXT(\"Away\")"
     Index:EGameReadiness Enum=@n2.GameReadiness
    ) -> (ReturnValue:Text)
   )
   > n3 = Branch(Condition="Equal (Enum)"(A:EGameReadiness Enum=@n2.GameReadiness, B:EGameReadiness Enum=Ready))
   > n4 = Branch(Condition=Get IsPlayerReady() -> (IsPlayerReady:bool))
   StopAnimation(execute=@n4.else, InAnimation=Get Ready() -> (Ready:Widget Animation Object Reference))
   n5 = Get Ready()
   > PlayAnimation(InAnimation=@n5)
   > PlaySound2D(Sound=/Game/Sounds/Fort_UI_Sounds/FrontEnd/SC_UI_MatchMaking_ReadyUp_Cue)
   > Set IsPlayerReady(IsPlayerReady:bool=true) -> (Output_Get:bool)
   Branch(execute=@n3.else, Condition=Get IsPlayerReady())
   > StopAnimation(InAnimation=Get Ready())
   > PlayAnimationReverse(InAnimation=@n5, PlaybackSpeed=2)
   > Set IsPlayerReady(IsPlayerReady=false)
  ]}
]
variables [
 IsPlayerReady {type=boolean}
 notPlayer {type=boolean}
]
widgets {
 info {inheritedWidgetCount=3, parentClass=/Script/FortniteUI.AthenaIGMPartyMemberWidget, rootWidgetClass=&VerticalBox, widgetCount=26}
 widgets [
  VerticalBox_1 {widgetClassPath=&VerticalBox, widget=VerticalBox_1}
  ScaleBox_ItemSize {
   widgetClassPath=&ScaleBox
   bIsVariable=true
   parent=VerticalBox_1
   widget=ScaleBox_ItemSize
   properties {stretch=UserSpecified}
   slot {class=VerticalBoxSlot}
  }
  VerticalBox_3 {widgetClassPath=&VerticalBox, parent=ScaleBox_ItemSize, widget=VerticalBox_3, slot {class=ScaleBoxSlot}}
  SizeBoxReadyText {
   widgetClassPath=&SizeBox
   parent=VerticalBox_3
   widget=SizeBoxReadyText
   properties {
    bOverride_WidthOverride=true
    renderOpacity=0
    renderTransform {angle=0, scale=<1 1>, shear=<0 0>, translation=<0 137.765>}
    renderTransformPivot=<0.5 1>
   }
   slot {class=VerticalBoxSlot, properties {verticalAlignment=VAlign_Bottom}}
  }
  ScaleBox_0 {
   widgetClassPath=&ScaleBox
   parent=SizeBoxReadyText
   widget=ScaleBox_0
   properties {stretchDirection=DownOnly}
   slot {class=SizeBoxSlot, properties {verticalAlignment=VAlign_Bottom}}
  }
  Text_Readiness {
   widgetClassPath=&CommonTextBlock
   bIsVariable=true
   parent=ScaleBox_0
   widget=Text_Readiness
   properties {
    justification=Center
    text="Ready!"
    textTransformPolicy=ToUpper
    font {
     bForceMonospaced=false
     bMaterialIsStencil=false
     fontMaterial=null
     fontObject=&Burbank
     letterSpacing=0
     monospacedWidth=1
     size=50
     skewAmount=0
     typefaceFontName=Burbank-PlayerPower
     outlineSettings {
      bApplyOutlineToDropShadows=false
      bMiteredCorners=false
      bSeparateFillAlpha=false
      outlineMaterial=null
      outlineSize=1
      outlineColor=<rgba 0.0052 0.0052 0.0056 0.2>
     }
    }
    renderTransform {angle=0, scale=<1 1>, shear=<-8 0>, translation=<0 0>}
    renderTransformPivot=<0.5 1>
    shadowOffset=<0 0>
   }
   slot {class=ScaleBoxSlot, properties {verticalAlignment=VAlign_Bottom}}
  }
  Overlay_0 {widgetClassPath=&Overlay, parent=VerticalBox_3, widget=Overlay_0, slot {class=VerticalBoxSlot}}
  SizeBox_2 {
   widgetClassPath=&SizeBox
   parent=Overlay_0
   widget=SizeBox_2
   properties {bOverride_HeightOverride=true, bOverride_WidthOverride=true, heightOverride=220, widthOverride=220}
   slot {class=OverlaySlot, properties {horizontalAlignment=HAlign_Center, verticalAlignment=VAlign_Center}}
  }
  Image_Selected_Effect {
   widgetClassPath=&Image
   bIsVariable=true
   parent=SizeBox_2
   widget=Image_Selected_Effect
   properties {
    renderOpacity=0
    visibility=HitTestInvisible
    brush {
     bIsDynamicallyLoaded=false
     drawAs=Box
     imageType=NoImage
     mirroring=NoMirror
     resourceName=None
     resourceObject=/Game/UI/Frontend/Showdown/Materials/M_UI_SkewBoxBorder_TournamentPanelBorder
     tiling=NoTile
     imageSize=<128 128>
     margin=<ltrb 0.2 0.3 0.2 0.3>
     outlineSettings {
      bUseBrushTransparency=false
      roundingType=HalfHeightRadius
      width=0
      color {colorUseRule=UseColor_Specified, specifiedColor=<rgba 0 0 0 0>}
      cornerRadii=<0 0 0 0>
     }
     tintColor {colorUseRule=UseColor_Specified, specifiedColor=<rgba 1 1 1 1>}
     uVRegion {bIsValid=false, max=<0 0>, min=<0 0>}
    }
   }
   slot {class=SizeBoxSlot}
  }
  Image_Selected {
   widgetClassPath=&Image
   bIsVariable=true
   parent=Overlay_0
   widget=Image_Selected
   properties {
    visibility=Collapsed
    brush {
     bIsDynamicallyLoaded=false
     drawAs=Image
     imageType=NoImage
     mirroring=NoMirror
     resourceName=None
     resourceObject=/Game/UI/Frontend/Showdown/Materials/M_UI_SkewBox_SelectedBG
     tiling=NoTile
     imageSize=<32 32>
     margin=<ltrb 0.2 0.2 0.2 0.2>
     outlineSettings {
      bUseBrushTransparency=false
      roundingType=HalfHeightRadius
      width=0
      color {colorUseRule=UseColor_Specified, specifiedColor=<rgba 0 0 0 0>}
      cornerRadii=<0 0 0 0>
     }
     tintColor {colorUseRule=UseColor_Specified, specifiedColor=<rgba 1 1 1 1>}
     uVRegion {bIsValid=false, max=<0 0>, min=<0 0>}
    }
    renderTransform {angle=0, scale=<1 1>, shear=<2 0>, translation=<0 0>}
   }
   slot {class=OverlaySlot, properties {horizontalAlignment=HAlign_Fill, verticalAlignment=VAlign_Fill, padding=<ltrb -12 -10 -16 -10>}}
  }
  SizeBox_0 {
   widgetClassPath=&SizeBox
   parent=Overlay_0
   widget=SizeBox_0
   properties {bOverride_HeightOverride=true, bOverride_WidthOverride=true, heightOverride=220, widthOverride=220}
   slot {class=OverlaySlot, properties {horizontalAlignment=HAlign_Fill, verticalAlignment=VAlign_Fill}}
  }
  Overlay_2 {widgetClassPath=&Overlay, parent=SizeBox_0, widget=Overlay_2, slot {class=SizeBoxSlot}}
  AvatarIconWidget {
   widgetClassPath=/Script/FortniteUI.FortSocialAvatarIcon
   bInherited=true
   bIsVariable=true
   parent=Overlay_2
   widget=AvatarIconWidget
   properties {
    avatarMaterial=/Game/UI/SocialPanel/Art/Materials/UI_UserPortrait_Face_IGM
    brush {
     bIsDynamicallyLoaded=false
     drawAs=Image
     imageType=NoImage
     mirroring=NoMirror
     resourceName=None
     resourceObject=&UI_UserPortrait_Face_IGM
     tiling=NoTile
     imageSize=<200 200>
     margin=<ltrb 0 0 0 0>
     outlineSettings {
      bUseBrushTransparency=false
      roundingType=HalfHeightRadius
      width=0
      color {colorUseRule=UseColor_Specified, specifiedColor=<rgba 0 0 0 0>}
      cornerRadii=<0 0 0 0>
     }
     tintColor {colorUseRule=UseColor_Specified, specifiedColor=<rgba 1 1 1 1>}
     uVRegion {bIsValid=false, max=<0 0>, min=<0 0>}
    }
   }
   slot {class=OverlaySlot, properties {horizontalAlignment=HAlign_Fill, verticalAlignment=VAlign_Fill}}
  }
  Overlay_AvatarExtension {
   widgetClassPath=&Overlay
   bIsVariable=true
   parent=Overlay_2
   widget=Overlay_AvatarExtension
   slot {class=OverlaySlot, properties {horizontalAlignment=HAlign_Fill, verticalAlignment=VAlign_Fill}}
  }
  CommonBorder_0 {
   widgetClassPath=&CommonBorder
   parent=Overlay_0
   widget=CommonBorder_0
   properties {
    background {
     bIsDynamicallyLoaded=false
     drawAs=Image
     imageType=NoImage
     mirroring=NoMirror
     resourceName=None
     resourceObject=/Game/Athena/HUD/Materials/M_UI_HUD_ReadyBG_NotReady
     tiling=NoTile
     imageSize=<32 32>
     margin=<ltrb 0 0 0 0>
     outlineSettings {
      bUseBrushTransparency=false
      roundingType=HalfHeightRadius
      width=0
      color {colorUseRule=UseColor_Specified, specifiedColor=<rgba 0 0 0 0>}
      cornerRadii=<0 0 0 0>
     }
     tintColor {colorUseRule=UseColor_Specified, specifiedColor=<rgba 1 1 1 1>}
     uVRegion {bIsValid=false, max=<0 0>, min=<0 0>}
    }
    padding=<ltrb 0 5 0 0>
   }
   slot {class=OverlaySlot, properties {horizontalAlignment=HAlign_Fill, verticalAlignment=VAlign_Bottom}}
  }
  Overlay_PartyLeaderIcon {
   widgetClassPath=&Overlay
   bIsVariable=true
   parent=Overlay_0
   widget=Overlay_PartyLeaderIcon
   properties {visibility=Collapsed, renderTransform {angle=-7, scale=<1 1>, shear=<0 0>, translation=<-19 6>}}
   slot {class=OverlaySlot, properties {verticalAlignment=VAlign_Bottom}}
  }
  PartyLeaderIconShadow {
   widgetClassPath=&Image
   parent=Overlay_PartyLeaderIcon
   widget=PartyLeaderIconShadow
   properties {
    visibility=SelfHitTestInvisible
    brush {
     bIsDynamicallyLoaded=false
     drawAs=Image
     imageType=NoImage
     mirroring=NoMirror
     resourceName=None
     resourceObject=/Game/UI/Icons/Icon-Nodes/Icon-SquadLeaderSlot-Generic-XL
     tiling=NoTile
     imageSize=<64 64>
     margin=<ltrb 0 0 0 0>
     outlineSettings {
      bUseBrushTransparency=false
      roundingType=HalfHeightRadius
      width=0
      color {colorUseRule=UseColor_Specified, specifiedColor=<rgba 0 0 0 0>}
      cornerRadii=<0 0 0 0>
     }
     tintColor {colorUseRule=UseColor_Specified, specifiedColor=<rgba 1 1 1 1>}
     uVRegion {bIsValid=false, max=<0 0>, min=<0 0>}
    }
    colorAndOpacity=<rgba 0.0194 0.0194 0.0666 0.8>
    renderTransform {angle=4, scale=<1 1>, shear=<0 0>, translation=<4 7>}
    renderTransformPivot=<0.3439 1>
   }
   slot {class=OverlaySlot}
  }
  PartyLeaderIcon {
   widgetClassPath=&Image
   parent=Overlay_PartyLeaderIcon
   widget=PartyLeaderIcon
   properties {
    visibility=SelfHitTestInvisible
    brush {
     bIsDynamicallyLoaded=false
     drawAs=Image
     imageType=NoImage
     mirroring=NoMirror
     resourceName=None
     resourceObject=&Icon-SquadLeaderSlot-Generic-XL
     tiling=NoTile
     imageSize=<64 64>
     margin=<ltrb 0 0 0 0>
     outlineSettings {
      bUseBrushTransparency=false
      roundingType=HalfHeightRadius
      width=0
      color {colorUseRule=UseColor_Specified, specifiedColor=<rgba 0 0 0 0>}
      cornerRadii=<0 0 0 0>
     }
     tintColor {colorUseRule=UseColor_Specified, specifiedColor=<rgba 1 1 1 1>}
     uVRegion {bIsValid=false, max=<0 0>, min=<0 0>}
    }
    colorAndOpacity=<rgba 1 0.8688 0 1>
   }
   slot {class=OverlaySlot}
  }
  Overlay_1 {
   widgetClassPath=&Overlay
   parent=Overlay_0
   widget=Overlay_1
   properties {renderTransform {angle=0, scale=<0 0>, shear=<0 0>, translation=<17 -2>}, renderTransformPivot=<0.5 1>}
   slot {class=OverlaySlot, properties {horizontalAlignment=HAlign_Right, verticalAlignment=VAlign_Bottom}}
  }
  ReadyCheckBox_Shadow {
   widgetClassPath=&Image
   bIsVariable=true
   parent=Overlay_1
   widget=ReadyCheckBox_Shadow
   properties {
    brush {
     bIsDynamicallyLoaded=false
     drawAs=Image
     imageType=NoImage
     mirroring=NoMirror
     resourceName=None
     resourceObject=/Game/UI/Foundation/Textures/Icons/Checks/T-Icon-Check-128
     tiling=NoTile
     imageSize=<68 68>
     margin=<ltrb 0 0 0 0>
     outlineSettings {
      bUseBrushTransparency=false
      roundingType=HalfHeightRadius
      width=0
      color {colorUseRule=UseColor_Specified, specifiedColor=<rgba 0 0 0 0>}
      cornerRadii=<0 0 0 0>
     }
     tintColor {colorUseRule=UseColor_Specified, specifiedColor=<rgba 1 1 1 1>}
     uVRegion {bIsValid=false, max=<0 0>, min=<0 0>}
    }
    colorAndOpacity=<rgba 0.0194 0.0194 0.0666 0.8>
    renderTransform {angle=4, scale=<1 1>, shear=<0 0>, translation=<4 7>}
    renderTransformPivot=<0.3439 1>
   }
   slot {class=OverlaySlot}
  }
  ReadyCheckBox {
   widgetClassPath=&Image
   bIsVariable=true
   parent=Overlay_1
   widget=ReadyCheckBox
   properties {
    brush {
     bIsDynamicallyLoaded=false
     drawAs=Image
     imageType=NoImage
     mirroring=NoMirror
     resourceName=None
     resourceObject=&T-Icon-Check-128
     tiling=NoTile
     imageSize=<68 68>
     margin=<ltrb 0 0 0 0>
     outlineSettings {
      bUseBrushTransparency=false
      roundingType=HalfHeightRadius
      width=0
      color {colorUseRule=UseColor_Specified, specifiedColor=<rgba 0 0 0 0>}
      cornerRadii=<0 0 0 0>
     }
     tintColor {colorUseRule=UseColor_Specified, specifiedColor=<rgba 1 1 1 1>}
     uVRegion {bIsValid=false, max=<0 0>, min=<0 0>}
    }
    colorAndOpacity=<rgba 0 1 0.1133 1>
   }
   slot {class=OverlaySlot}
  }
  SizeBox_1 {
   widgetClassPath=&SizeBox
   parent=VerticalBox_3
   widget=SizeBox_1
   properties {bOverride_WidthOverride=true}
   slot {class=VerticalBoxSlot}
  }
  ScaleBox {
   widgetClassPath=&ScaleBox
   parent=SizeBox_1
   widget=ScaleBox
   slot {class=SizeBoxSlot, properties {horizontalAlignment=HAlign_Left}}
  }
  HorizontalBox_0 {widgetClassPath=&HorizontalBox, parent=ScaleBox, widget=HorizontalBox_0, slot {class=ScaleBoxSlot}}
  Text_SocialName {
   widgetClassPath=/Script/FortniteUI.FortSocialNameTextBlock
   bInherited=true
   bIsVariable=true
   parent=HorizontalBox_0
   widget=Text_SocialName
   properties {
    bOverrideDefaultStyle=true
    bTintInlineIcon=true
    text="User Name<PC/> <text color=\"1DEFA9FF\" size=\"12\">(NickName)</>"
    defaultTextStyleOverride {
     overflowPolicy=Clip
     transformPolicy=None
     colorAndOpacity {colorUseRule=UseColor_Specified, specifiedColor=<rgba 1 1 1 1>}
     font {
      bForceMonospaced=false
      bMaterialIsStencil=false
      fontMaterial=null
      fontObject=&Burbank
      letterSpacing=0
      monospacedWidth=1
      size=14
      skewAmount=0
      typefaceFontName=BurbankRegular-Black
      outlineSettings {
       bApplyOutlineToDropShadows=true
       bMiteredCorners=false
       bSeparateFillAlpha=false
       outlineMaterial=null
       outlineSize=1
       outlineColor=<rgba 0 0 0 0.75>
      }
     }
     highlightColor {colorUseRule=UseColor_Specified, specifiedColor=<rgba 0 0 0 1>}
     highlightShape {
      bIsDynamicallyLoaded=false
      drawAs=Image
      imageType=NoImage
      mirroring=NoMirror
      resourceName=None
      resourceObject=null
      tiling=NoTile
      imageSize=<32 32>
      margin=<ltrb 0 0 0 0>
      outlineSettings {
       bUseBrushTransparency=false
       roundingType=HalfHeightRadius
       width=0
       color {colorUseRule=UseColor_Specified, specifiedColor=<rgba 0 0 0 0>}
       cornerRadii=<0 0 0 0>
      }
      tintColor {colorUseRule=UseColor_Specified, specifiedColor=<rgba 1 1 1 1>}
      uVRegion {bIsValid=false, max=<0 0>, min=<0 0>}
     }
     selectedBackgroundColor {colorUseRule=UseColor_Foreground, specifiedColor=<rgba 1 0 1 1>}
     shadowColorAndOpacity=<rgba 0 0 0 1>
     shadowOffset=<1 1>
     strikeBrush {
      bIsDynamicallyLoaded=false
      drawAs=Image
      imageType=NoImage
      mirroring=NoMirror
      resourceName=None
      resourceObject=null
      tiling=NoTile
      imageSize=<32 32>
      margin=<ltrb 0 0 0 0>
      outlineSettings {
       bUseBrushTransparency=false
       roundingType=HalfHeightRadius
       width=0
       color {colorUseRule=UseColor_Specified, specifiedColor=<rgba 0 0 0 0>}
       cornerRadii=<0 0 0 0>
      }
      tintColor {colorUseRule=UseColor_Specified, specifiedColor=<rgba 1 1 1 1>}
      uVRegion {bIsValid=false, max=<0 0>, min=<0 0>}
     }
     underlineBrush {
      bIsDynamicallyLoaded=false
      drawAs=Image
      imageType=NoImage
      mirroring=NoMirror
      resourceName=None
      resourceObject=null
      tiling=NoTile
      imageSize=<32 32>
      margin=<ltrb 0 0 0 0>
      outlineSettings {
       bUseBrushTransparency=false
       roundingType=HalfHeightRadius
       width=0
       color {colorUseRule=UseColor_Specified, specifiedColor=<rgba 0 0 0 0>}
       cornerRadii=<0 0 0 0>
      }
      tintColor {colorUseRule=UseColor_Specified, specifiedColor=<rgba 1 1 1 1>}
      uVRegion {bIsValid=false, max=<0 0>, min=<0 0>}
     }
    }
    renderTransform {angle=0, scale=<1 1>, shear=<-8 0>, translation=<0 0>}
   }
   slot {class=HorizontalBoxSlot, properties {verticalAlignment=VAlign_Center}}
  }
  VoiceIcon_Status {
   widgetClassPath=/Script/FortniteUI.FortVoiceChatStatusIconProxyAnim
   bInherited=true
   bIsVariable=true
   parent=HorizontalBox_0
   widget=VoiceIcon_Status
   slot {class=HorizontalBoxSlot, properties {horizontalAlignment=HAlign_Left, verticalAlignment=VAlign_Center, padding=<ltrb 3 0 -60 0>}}
  }
 ]
}
