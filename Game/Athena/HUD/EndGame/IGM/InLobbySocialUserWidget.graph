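# Party member card in the end-game IGM HUD, showing avatar, name, voice status, leader crown, location and readiness.
# Parent-class events show or collapse the card, toggle the leader icon and map party location to localized text.
# Readiness switches the Not Ready/Ready/Sitting Out label and plays the Ready animation forward or reversed with matching sounds.
# The animation pops the card's scale, fades in the checkmark and recolors the avatar background material.
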WidgetBlueprint /Game/Athena/HUD/EndGame/IGM/InLobbySocialUserWidget
animations [Ready {
  binding_count=4
  length_seconds=0.3
  object=&self:Ready
  sequence {bindings [
    Overlay_readycheck {
     id=#1
     possessed_class=Overlay
     tracks [{
       type=MovieSceneFloatTrack
       display_name=Render Opacity
       sections [{type=MovieSceneFloatSection, channels [None {type=MovieSceneScriptingFloatChannel, keys [[0,0],[1.6287,1]]}]}]
      }]
    }
    Image_BG {id=#2, possessed_class=Image, tracks [{type=MovieSceneWidgetMaterialTrack, sections []}]}
    Overlay_2 {
     id=#3
     possessed_class=Overlay
     tracks [{
       type=MovieScene2DTransformTrack
       sections [{
         type=MovieScene2DTransformSection
         channels [
          Scale.X {type=MovieSceneScriptingFloatChannel, keys [[0,1],[1.1457,1.3206],[4.2157,0.9631],[6,1]]}
          Scale.Y {type=MovieSceneScriptingFloatChannel, keys [[0,1],[1.1457,1.3206],[4.2157,0.9631],[6,1]]}
         ]
        }]
      }]
    }
    CommonBorderAvatar {
     id=#4
     display_name=CommonBorder_0
     possessed_class=CommonBorder
     tracks [{
       type=MovieSceneWidgetMaterialTrack
       sections [{
         type=MovieSceneParameterSection
         channels [
          BGColorR {type=MovieSceneScriptingFloatChannel, keys [[0,0.3419],[6,0.0908]]}
          BGColorG {type=MovieSceneScriptingFloatChannel, keys [[0,0],[6,0.5333]]}
          BGColorB {type=MovieSceneScriptingFloatChannel, keys [[0,0.5647],[6,1]]}
          BGColorA {type=MovieSceneScriptingFloatChannel, const [1,2]}
         ]
        }]
      }]
    }
   ]}
 }]
graphs [EventGraph {nodes [
   n1 = Event OnIsPartyLeaderChanged!() -> (OutputDelegate, then, bIsPartyLeader:bool) {is=title K2Node_Event}
   > SetVisibility(
    self=Get Overlay_PartyLeaderIcon() -> (Overlay_PartyLeaderIcon:Overlay Object Reference)
    InVisibility=Select(Option 0=Collapsed, Option 1=SelfHitTestInvisible, Index=@n1.bIsPartyLeader)
   )
   Event OnUserSet() {is=title K2Node_Event}
   > SetVisibility(InVisibility=HitTestInvisible)
   Event OnWidgetReset() {is=title K2Node_Event}
   > SetVisibility(InVisibility=Collapsed)
   > SetVisibility(self=Get Overlay_PartyLeaderIcon(), InVisibility=Collapsed)
   n2 = Event OnNewLocation!() -> (OutputDelegate, then, NewLocation:EFortPartyMemberLocation Enum=PreLobby) {is=title K2Node_Event}
   > "SetText (Text)"(
    self=Get Text_Location() -> (Text_Location:Common Text Object Reference)
    InText=Select!(
     PreLobby:Text="LOCTEXT(\"Lobby\")"
     ConnectingToLobby:Text="LOCTEXT(\"Connecting to Lobby\")"
     Lobby:Text="LOCTEXT(\"Lobby\")"
     JoiningGame:Text="LOCTEXT(\"Joining Game\")"
     ProcessingRejoin:Text="LOCTEXT(\"Rejoining\")"
     InGame:Text="LOCTEXT(\"In Game\")"
     Spectating:Text="LOCTEXT(\"Spectating\")"
     WatchingReplay:Text="LOCTEXT(\"Watching Replay\")"
     ReturningToFrontEnd:Text="LOCTEXT(\"Returning to Lobby\")"
     Index:EFortPartyMemberLocation Enum=@n2.NewLocation
    ) -> (ReturnValue:Text)
   )
   n3 = Event OnNewReadiness!() -> (OutputDelegate, then, GameReadiness:EGameReadiness Enum=NotReady) {is=title K2Node_Event}
   > SetActiveWidgetIndex(
    self=Get CommonWidgetSwitcher_ReadyState() -> (CommonWidgetSwitcher_ReadyState:Common Widget Switcher Legacy Object Reference)
    Index=Select!(NotReady:int=0, Ready:int=1, Reevaluating:int=0, SittingOut:int=2, Away:int=0, Index:EGameReadiness Enum=@n3.GameReadiness) -> (ReturnValue:int)
   )
   > n4 = Branch(Condition="Equal (Enum)"(A:EGameReadiness Enum=@n3.GameReadiness, B:EGameReadiness Enum=Ready))
   n5 = Get Ready() -> (Ready:Widget Animation Object Reference)
   > PlayAnimation(InAnimation=@n5)
   > PlaySound2D(Sound=/Game/Sounds/Fort_UI_Sounds/FrontEnd/SC_UI_MatchMaking_ReadyUp_Cue)
   PlayAnimationReverse(execute=@n4.else, InAnimation=@n5)
   > PlaySound2D(Sound=/Game/Sounds/Fort_UI_Sounds/FrontEnd/SC_UI_MatchMaking_UnReady_Cue)
   SetVisibility(
    self=Get ReadyCheckBox() -> (ReadyCheckBox:Image Object Reference)
    InVisibility=Select(
     NotReady:ESlateVisibility Enum=Hidden
     Ready:ESlateVisibility Enum=HitTestInvisible
     Reevaluating:ESlateVisibility Enum=Visible
     SittingOut:ESlateVisibility Enum=Hidden
     Away:ESlateVisibility Enum=Visible
     Index=@n3.GameReadiness
    ) -> (ReturnValue:ESlateVisibility Enum)
   )
  ]}]
variables []
widgets {
 info {inheritedWidgetCount=3, parentClass=/Script/FortniteUI.AthenaIGMPartyMemberWidget, rootWidgetClass=&Overlay, widgetCount=29}
 widgets [
  Overlay_3 {widgetClassPath=&Overlay, widget=Overlay_3}
  Overlay_2 {
   widgetClassPath=&Overlay
   parent=Overlay_3
   widget=Overlay_2
   properties {renderTransformPivot=<0 0.5>}
   slot {class=OverlaySlot, properties {padding=<ltrb 0 0 0 25>}}
  }
  Image_BG {
   widgetClassPath=&Image
   bIsVariable=true
   parent=Overlay_2
   widget=Image_BG
   properties {
    brush {
     bIsDynamicallyLoaded=false
     drawAs=Box
     imageType=NoImage
     mirroring=NoMirror
     resourceName=None
     resourceObject=/Game/UI/Frontend/Showdown/Materials/M_UI_SkewBox_LeaderboardDetailsBG2
     tiling=NoTile
     imageSize=<32 32>
     margin=<ltrb 0.2 0.2 0.2 0.2>
     outlineSettings {
      bUseBrushTransparency=false
      roundingType=HalfHeightRadius
      width=0
      color {colorUseRule=UseColor_Specified, specifiedColor=<rgba 0 0 0 0>}
      cornerRadii=<0 0 0 0>
     }
     tintColor {colorUseRule=UseColor_Specified, specifiedColor=<rgba 1 1 1 1>}
     uVRegion {bIsValid=false, max=<0 0>, min=<0 0>}
    }
    colorAndOpacity=<rgba 0.0052 0.0052 0.0056 0.5>
   }
   slot {class=OverlaySlot, properties {horizontalAlignment=HAlign_Fill, verticalAlignment=VAlign_Fill, padding=<ltrb 0 10 0 10>}}
  }
  HorizontalBox_0 {
   widgetClassPath=&HorizontalBox
   parent=Overlay_2
   widget=HorizontalBox_0
   slot {class=OverlaySlot, properties {verticalAlignment=VAlign_Center}}
  }
  Overlay_0 {
   widgetClassPath=&Overlay
   parent=HorizontalBox_0
   widget=Overlay_0
   slot {class=HorizontalBoxSlot, properties {verticalAlignment=VAlign_Center, padding=<ltrb 20 0 0 0>}}
  }
  SizeBox_2 {
   widgetClassPath=&SizeBox
   parent=Overlay_0
   widget=SizeBox_2
   properties {bOverride_HeightOverride=true, bOverride_WidthOverride=true, heightOverride=110, widthOverride=110}
   slot {class=OverlaySlot, properties {horizontalAlignment=HAlign_Center, verticalAlignment=VAlign_Center}}
  }
  Image_Selected_Effect {
   widgetClassPath=&Image
   bIsVariable=true
   parent=SizeBox_2
   widget=Image_Selected_Effect
   properties {
    renderOpacity=0
    visibility=HitTestInvisible
    brush {
     bIsDynamicallyLoaded=false
     drawAs=Box
     imageType=NoImage
     mirroring=NoMirror
     resourceName=None
     resourceObject=/Game/UI/Frontend/Showdown/Materials/M_UI_SkewBoxBorder_TournamentPanelBorder
     tiling=NoTile
     imageSize=<128 128>
     margin=<ltrb 0.2 0.3 0.2 0.3>
     outlineSettings {
      bUseBrushTransparency=false
      roundingType=HalfHeightRadius
      width=0
      color {colorUseRule=UseColor_Specified, specifiedColor=<rgba 0 0 0 0>}
      cornerRadii=<0 0 0 0>
     }
     tintColor {colorUseRule=UseColor_Specified, specifiedColor=<rgba 1 1 1 1>}
     uVRegion {bIsValid=false, max=<0 0>, min=<0 0>}
    }
   }
   slot {class=SizeBoxSlot}
  }
  Image_Selected {
   widgetClassPath=&Image
   bIsVariable=true
   parent=Overlay_0
   widget=Image_Selected
   properties {
    visibility=Collapsed
    brush {
     bIsDynamicallyLoaded=false
     drawAs=Image
     imageType=NoImage
     mirroring=NoMirror
     resourceName=None
     resourceObject=/Game/UI/Frontend/Showdown/Materials/M_UI_SkewBox_SelectedBG
     tiling=NoTile
     imageSize=<32 32>
     margin=<ltrb 0 0 0 0>
     outlineSettings {
      bUseBrushTransparency=false
      roundingType=HalfHeightRadius
      width=0
      color {colorUseRule=UseColor_Specified, specifiedColor=<rgba 0 0 0 0>}
      cornerRadii=<0 0 0 0>
     }
     tintColor {colorUseRule=UseColor_Specified, specifiedColor=<rgba 1 1 1 1>}
     uVRegion {bIsValid=false, max=<0 0>, min=<0 0>}
    }
    renderTransform {angle=0, scale=<1 1>, shear=<2 0>, translation=<0 0>}
   }
   slot {class=OverlaySlot, properties {horizontalAlignment=HAlign_Fill, verticalAlignment=VAlign_Fill, padding=<ltrb -12 -10 -16 -10>}}
  }
  Overlay_1 {
   widgetClassPath=&Overlay
   parent=Overlay_0
   widget=Overlay_1
   slot {class=OverlaySlot, properties {horizontalAlignment=HAlign_Center, verticalAlignment=VAlign_Center}}
  }
  CommonBorderAvatar {
   widgetClassPath=&CommonBorder
   parent=Overlay_1
   widget=CommonBorderAvatar
   properties {
    background {
     bIsDynamicallyLoaded=false
     drawAs=Image
     imageType=NoImage
     mirroring=NoMirror
     resourceName=None
     resourceObject=/Game/Athena/HUD/Materials/M_UI_HUD_ReadyBG_NotReady
     tiling=NoTile
     imageSize=<32 32>
     margin=<ltrb 0 0 0 0>
     outlineSettings {
      bUseBrushTransparency=false
      roundingType=HalfHeightRadius
      width=0
      color {colorUseRule=UseColor_Specified, specifiedColor=<rgba 0 0 0 0>}
      cornerRadii=<0 0 0 0>
     }
     tintColor {colorUseRule=UseColor_Specified, specifiedColor=<rgba 1 1 1 1>}
     uVRegion {bIsValid=false, max=<0 0>, min=<0 0>}
    }
    padding=<ltrb 0 0 0 0>
   }
   slot {class=OverlaySlot}
  }
  SizeBox_0 {
   widgetClassPath=&SizeBox
   parent=CommonBorderAvatar
   widget=SizeBox_0
   properties {bOverride_HeightOverride=true, bOverride_WidthOverride=true, heightOverride=110, widthOverride=110}
   slot {class=BorderSlot, properties {padding=<ltrb 0 0 0 0>}}
  }
  AvatarIconWidget {
   widgetClassPath=/Script/FortniteUI.FortSocialAvatarIcon
   bInherited=true
   bIsVariable=true
   parent=SizeBox_0
   widget=AvatarIconWidget
   properties {
    avatarMaterial=/Game/UI/SocialPanel/Art/Materials/UI_UserPortrait_Face_IGM
    brush {
     bIsDynamicallyLoaded=false
     drawAs=Image
     imageType=NoImage
     mirroring=NoMirror
     resourceName=None
     resourceObject=&UI_UserPortrait_Face_IGM
     tiling=NoTile
     imageSize=<200 200>
     margin=<ltrb 0 0 0 0>
     outlineSettings {
      bUseBrushTransparency=false
      roundingType=HalfHeightRadius
      width=0
      color {colorUseRule=UseColor_Specified, specifiedColor=<rgba 0 0 0 0>}
      cornerRadii=<0 0 0 0>
     }
     tintColor {colorUseRule=UseColor_Specified, specifiedColor=<rgba 1 1 1 1>}
     uVRegion {bIsValid=false, max=<0 0>, min=<0 0>}
    }
   }
   slot {class=SizeBoxSlot}
  }
  SizeBox_4 {
   widgetClassPath=&SizeBox
   parent=Overlay_1
   widget=SizeBox_4
   properties {bOverride_WidthOverride=true, renderTransform {angle=0, scale=<1 1>, shear=<0 0>, translation=<0 29>}}
   slot {class=OverlaySlot, properties {verticalAlignment=VAlign_Bottom}}
  }
  HorizontalBox_1 {widgetClassPath=&HorizontalBox, parent=SizeBox_4, widget=HorizontalBox_1, slot {class=SizeBoxSlot}}
  Text_SocialName {
   widgetClassPath=/Script/FortniteUI.FortSocialNameTextBlock
   bInherited=true
   bIsVariable=true
   parent=HorizontalBox_1
   widget=Text_SocialName
   properties {
    bOverrideDefaultStyle=true
    bTintInlineIcon=true
    text="User Name<PC/> <text color=\"1DEFA9FF\" size=\"12\">(NickName)</>"
    defaultTextStyleOverride {
     overflowPolicy=Clip
     transformPolicy=None
     colorAndOpacity {colorUseRule=UseColor_Specified, specifiedColor=<rgba 1 1 1 1>}
     font {
      bForceMonospaced=false
      bMaterialIsStencil=false
      fontMaterial=null
      fontObject=&Burbank
      letterSpacing=0
      monospacedWidth=1
      size=15
      skewAmount=0
      typefaceFontName=Burbank-PlayerPower
      outlineSettings {
       bApplyOutlineToDropShadows=false
       bMiteredCorners=false
       bSeparateFillAlpha=false
       outlineMaterial=null
       outlineSize=0
       outlineColor=<rgba 0 0 0 1>
      }
     }
     highlightColor {colorUseRule=UseColor_Specified, specifiedColor=<rgba 0 0 0 1>}
     highlightShape {
      bIsDynamicallyLoaded=false
      drawAs=Image
      imageType=NoImage
      mirroring=NoMirror
      resourceName=None
      resourceObject=null
      tiling=NoTile
      imageSize=<32 32>
      margin=<ltrb 0 0 0 0>
      outlineSettings {
       bUseBrushTransparency=false
       roundingType=HalfHeightRadius
       width=0
       color {colorUseRule=UseColor_Specified, specifiedColor=<rgba 0 0 0 0>}
       cornerRadii=<0 0 0 0>
      }
      tintColor {colorUseRule=UseColor_Specified, specifiedColor=<rgba 1 1 1 1>}
      uVRegion {bIsValid=false, max=<0 0>, min=<0 0>}
     }
     selectedBackgroundColor {colorUseRule=UseColor_Foreground, specifiedColor=<rgba 1 0 1 1>}
     shadowColorAndOpacity=<rgba 0 0 0 1>
     shadowOffset=<1 1>
     strikeBrush {
      bIsDynamicallyLoaded=false
      drawAs=Image
      imageType=NoImage
      mirroring=NoMirror
      resourceName=None
      resourceObject=null
      tiling=NoTile
      imageSize=<32 32>
      margin=<ltrb 0 0 0 0>
      outlineSettings {
       bUseBrushTransparency=false
       roundingType=HalfHeightRadius
       width=0
       color {colorUseRule=UseColor_Specified, specifiedColor=<rgba 0 0 0 0>}
       cornerRadii=<0 0 0 0>
      }
      tintColor {colorUseRule=UseColor_Specified, specifiedColor=<rgba 1 1 1 1>}
      uVRegion {bIsValid=false, max=<0 0>, min=<0 0>}
     }
     underlineBrush {
      bIsDynamicallyLoaded=false
      drawAs=Image
      imageType=NoImage
      mirroring=NoMirror
      resourceName=None
      resourceObject=null
      tiling=NoTile
      imageSize=<32 32>
      margin=<ltrb 0 0 0 0>
      outlineSettings {
       bUseBrushTransparency=false
       roundingType=HalfHeightRadius
       width=0
       color {colorUseRule=UseColor_Specified, specifiedColor=<rgba 0 0 0 0>}
       cornerRadii=<0 0 0 0>
      }
      tintColor {colorUseRule=UseColor_Specified, specifiedColor=<rgba 1 1 1 1>}
      uVRegion {bIsValid=false, max=<0 0>, min=<0 0>}
     }
    }
    renderTransform {angle=0, scale=<1 1>, shear=<-8 0>, translation=<0 0>}
   }
   slot {class=HorizontalBoxSlot, properties {verticalAlignment=VAlign_Center}}
  }
  VoiceIcon_Status {
   widgetClassPath=/Script/FortniteUI.FortVoiceChatStatusIconProxyAnim
   bInherited=true
   bIsVariable=true
   parent=HorizontalBox_1
   widget=VoiceIcon_Status
   slot {class=HorizontalBoxSlot, properties {horizontalAlignment=HAlign_Left, verticalAlignment=VAlign_Center, padding=<ltrb 5 0 0 0>}}
  }
  Overlay_PartyLeaderIcon {
   widgetClassPath=&Overlay
   bIsVariable=true
   parent=Overlay_0
   widget=Overlay_PartyLeaderIcon
   properties {visibility=Collapsed, renderTransform {angle=-7, scale=<1 1>, shear=<0 0>, translation=<-12 2>}}
   slot {class=OverlaySlot, properties {verticalAlignment=VAlign_Bottom}}
  }
  PartyLeaderIconShadow {
   widgetClassPath=&Image
   parent=Overlay_PartyLeaderIcon
   widget=PartyLeaderIconShadow
   properties {
    visibility=SelfHitTestInvisible
    brush {
     bIsDynamicallyLoaded=false
     drawAs=Image
     imageType=NoImage
     mirroring=NoMirror
     resourceName=None
     resourceObject=/Game/UI/Icons/Icon-Nodes/Icon-SquadLeaderSlot-Generic-XL
     tiling=NoTile
     imageSize=<34 34>
     margin=<ltrb 0 0 0 0>
     outlineSettings {
      bUseBrushTransparency=false
      roundingType=HalfHeightRadius
      width=0
      color {colorUseRule=UseColor_Specified, specifiedColor=<rgba 0 0 0 0>}
      cornerRadii=<0 0 0 0>
     }
     tintColor {colorUseRule=UseColor_Specified, specifiedColor=<rgba 1 1 1 1>}
     uVRegion {bIsValid=false, max=<0 0>, min=<0 0>}
    }
    colorAndOpacity=<rgba 0.0194 0.0194 0.0666 0.8>
    renderTransform {angle=0, scale=<1 1>, shear=<0 0>, translation=<2 3>}
    renderTransformPivot=<0.3439 1>
   }
   slot {class=OverlaySlot}
  }
  PartyLeaderIcon {
   widgetClassPath=&Image
   parent=Overlay_PartyLeaderIcon
   widget=PartyLeaderIcon
   properties {
    visibility=SelfHitTestInvisible
    brush {
     bIsDynamicallyLoaded=false
     drawAs=Image
     imageType=NoImage
     mirroring=NoMirror
     resourceName=None
     resourceObject=&Icon-SquadLeaderSlot-Generic-XL
     tiling=NoTile
     imageSize=<34 34>
     margin=<ltrb 0 0 0 0>
     outlineSettings {
      bUseBrushTransparency=false
      roundingType=HalfHeightRadius
      width=0
      color {colorUseRule=UseColor_Specified, specifiedColor=<rgba 0 0 0 0>}
      cornerRadii=<0 0 0 0>
     }
     tintColor {colorUseRule=UseColor_Specified, specifiedColor=<rgba 1 1 1 1>}
     uVRegion {bIsValid=false, max=<0 0>, min=<0 0>}
    }
    colorAndOpacity=<rgba 1 0.8688 0 1>
   }
   slot {class=OverlaySlot}
  }
  Overlay_readycheck {
   widgetClassPath=&Overlay
   parent=Overlay_0
   widget=Overlay_readycheck
   properties {renderTransform {angle=0, scale=<1 1>, shear=<0 0>, translation=<7 -2>}, renderTransformPivot=<0.5 1>}
   slot {class=OverlaySlot, properties {horizontalAlignment=HAlign_Right, verticalAlignment=VAlign_Bottom}}
  }
  ReadyCheckBox_Shadow {
   widgetClassPath=&Image
   bIsVariable=true
   parent=Overlay_readycheck
   widget=ReadyCheckBox_Shadow
   properties {
    brush {
     bIsDynamicallyLoaded=false
     drawAs=Image
     imageType=NoImage
     mirroring=NoMirror
     resourceName=None
     resourceObject=/Game/UI/Foundation/Textures/Icons/Checks/T-Icon-Check-128
     tiling=NoTile
     imageSize=<34 34>
     margin=<ltrb 0 0 0 0>
     outlineSettings {
      bUseBrushTransparency=false
      roundingType=HalfHeightRadius
      width=0
      color {colorUseRule=UseColor_Specified, specifiedColor=<rgba 0 0 0 0>}
      cornerRadii=<0 0 0 0>
     }
     tintColor {colorUseRule=UseColor_Specified, specifiedColor=<rgba 1 1 1 1>}
     uVRegion {bIsValid=false, max=<0 0>, min=<0 0>}
    }
    colorAndOpacity=<rgba 0.0194 0.0194 0.0666 0.8>
    renderTransform {angle=4, scale=<1 1>, shear=<0 0>, translation=<2 3>}
    renderTransformPivot=<0.3439 1>
   }
   slot {class=OverlaySlot}
  }
  ReadyCheckBox {
   widgetClassPath=&Image
   bIsVariable=true
   parent=Overlay_readycheck
   widget=ReadyCheckBox
   properties {
    brush {
     bIsDynamicallyLoaded=false
     drawAs=Image
     imageType=NoImage
     mirroring=NoMirror
     resourceName=None
     resourceObject=&T-Icon-Check-128
     tiling=NoTile
     imageSize=<34 34>
     margin=<ltrb 0 0 0 0>
     outlineSettings {
      bUseBrushTransparency=false
      roundingType=HalfHeightRadius
      width=0
      color {colorUseRule=UseColor_Specified, specifiedColor=<rgba 0 0 0 0>}
      cornerRadii=<0 0 0 0>
     }
     tintColor {colorUseRule=UseColor_Specified, specifiedColor=<rgba 1 1 1 1>}
     uVRegion {bIsValid=false, max=<0 0>, min=<0 0>}
    }
    colorAndOpacity=<rgba 0 1 0.1133 1>
   }
   slot {class=OverlaySlot}
  }
  SizeBox_1 {
   widgetClassPath=&SizeBox
   parent=HorizontalBox_0
   widget=SizeBox_1
   properties {bOverride_HeightOverride=true, bOverride_MinDesiredWidth=true, heightOverride=110, minDesiredWidth=365}
   slot {class=HorizontalBoxSlot, properties {verticalAlignment=VAlign_Center}}
  }
  VerticalBox_0 {
   widgetClassPath=&VerticalBox
   parent=SizeBox_1
   widget=VerticalBox_0
   slot {class=SizeBoxSlot, properties {verticalAlignment=VAlign_Center, padding=<ltrb 20 0 0 0>}}
  }
  Text_Location {
   widgetClassPath=&CommonTextBlock
   bIsVariable=true
   parent=VerticalBox_0
   widget=Text_Location
   properties {
    text=In Lobby
    textTransformPolicy=ToUpper
    font {
     bForceMonospaced=false
     bMaterialIsStencil=false
     fontMaterial=null
     fontObject=&Burbank
     letterSpacing=0
     monospacedWidth=1
     size=20
     skewAmount=0
     typefaceFontName=Burbank-PlayerPower
     outlineSettings {
      bApplyOutlineToDropShadows=false
      bMiteredCorners=false
      bSeparateFillAlpha=false
      outlineMaterial=null
      outlineSize=0
      outlineColor=<rgba 0 0 0 1>
     }
    }
    renderTransform {angle=0, scale=<1 1>, shear=<-8 0>, translation=<0 0>}
    shadowOffset=<0 0>
   }
   slot {class=VerticalBoxSlot, properties {padding=<ltrb 0 -3 0 3>}}
  }
  CommonWidgetSwitcher_ReadyState {
   widgetClassPath=&CommonWidgetSwitcherLegacy
   bIsVariable=true
   parent=VerticalBox_0
   widget=CommonWidgetSwitcher_ReadyState
   properties {transitionType=Horizontal}
   slot {class=VerticalBoxSlot}
  }
  Text_Readiness_2 {
   widgetClassPath=&CommonTextBlock
   parent=CommonWidgetSwitcher_ReadyState
   widget=Text_Readiness_2
   properties {
    text=Not Ready
    textTransformPolicy=ToUpper
    font {
     bForceMonospaced=false
     bMaterialIsStencil=false
     fontMaterial=null
     fontObject=&Burbank
     letterSpacing=0
     monospacedWidth=1
     size=20
     skewAmount=0
     typefaceFontName=Burbank-PlayerPower
     outlineSettings {
      bApplyOutlineToDropShadows=false
      bMiteredCorners=false
      bSeparateFillAlpha=false
      outlineMaterial=null
      outlineSize=1
      outlineColor=<rgba 0.1559 0 0.402 1>
     }
    }
    renderTransform {angle=0, scale=<1 1>, shear=<-8 0>, translation=<0 0>}
    shadowColorAndOpacity=<rgba 0.1559 0 0.402 1>
    shadowOffset=<2 2>
   }
   slot {class=WidgetSwitcherSlot, properties {padding=<ltrb 0 -5 0 -8>}}
  }
  Text_Readiness {
   widgetClassPath=&CommonTextBlock
   bIsVariable=true
   parent=CommonWidgetSwitcher_ReadyState
   widget=Text_Readiness
   properties {
    text="Ready!"
    textTransformPolicy=ToUpper
    font {
     bForceMonospaced=false
     bMaterialIsStencil=false
     fontMaterial=null
     fontObject=&Burbank
     letterSpacing=0
     monospacedWidth=1
     size=30
     skewAmount=0
     typefaceFontName=Burbank-PlayerPower
     outlineSettings {
      bApplyOutlineToDropShadows=false
      bMiteredCorners=false
      bSeparateFillAlpha=false
      outlineMaterial=null
      outlineSize=2
      outlineColor=<rgba 0 0.227 1 1>
     }
    }
    renderTransform {angle=0, scale=<1 1>, shear=<-8 0>, translation=<0 0>}
    shadowColorAndOpacity=<rgba 0.0908 0.5333 1 1>
    shadowOffset=<4 4>
   }
   slot {class=WidgetSwitcherSlot, properties {padding=<ltrb 0 -8 0 -12>}}
  }
  Text_Readiness_1 {
   widgetClassPath=&CommonTextBlock
   parent=CommonWidgetSwitcher_ReadyState
   widget=Text_Readiness_1
   properties {
    text=Sitting Out
    textTransformPolicy=ToUpper
    font {
     bForceMonospaced=false
     bMaterialIsStencil=false
     fontMaterial=null
     fontObject=&Burbank
     letterSpacing=0
     monospacedWidth=1
     size=20
     skewAmount=0
     typefaceFontName=Burbank-PlayerPower
     outlineSettings {
      bApplyOutlineToDropShadows=false
      bMiteredCorners=false
      bSeparateFillAlpha=false
      outlineMaterial=null
      outlineSize=1
      outlineColor=<rgba 0.1559 0 0.402 1>
     }
    }
    shadowColorAndOpacity=<rgba 0.1559 0 0.402 1>
    shadowOffset=<2 2>
   }
   slot {class=WidgetSwitcherSlot, properties {padding=<ltrb 0 -5 0 -8>}}
  }
 ]
}
