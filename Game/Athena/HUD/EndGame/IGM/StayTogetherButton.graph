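# Stay As A Team button widget for the Athena end-game HUD, derived from AthenaPreserveSquadButton, which shows whether squad preservation is on.
# An inherited quiet UIKit button, bound to the GenericRightShoulder input action, fills an overlay shared with a right-aligned check badge that starts collapsed.
# The badge layers a green check over a dark, slightly rotated copy as a drop shadow.
# Handle Preserve Squad Changed toggles the badge between visible and collapsed.
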WidgetBlueprint /Game/Athena/HUD/EndGame/IGM/StayTogetherButton
graphs [EventGraph {nodes [
   n1 = Event Handle Preserve Squad Changed!() -> (OutputDelegate, then, bPreserveSquad:bool) {is=title K2Node_Event}
   > SetVisibility(
    self=Get SB_Check() -> (SB_Check:Size Box Object Reference)
    InVisibility=Select(Option 0=Collapsed, Option 1=SelfHitTestInvisible, Index=@n1.bPreserveSquad)
   )
  ]}]
variables []
widgets {
 info {inheritedWidgetCount=1, parentClass=/Script/FortniteUI.AthenaPreserveSquadButton, rootWidgetClass=&Overlay, widgetCount=6}
 widgets [
  Overlay_1 {widgetClassPath=&Overlay, widget=Overlay_1}
  SB_Check {
   widgetClassPath=&SizeBox
   bIsVariable=true
   parent=Overlay_1
   widget=SB_Check
   properties {bOverride_WidthOverride=true, visibility=Collapsed, widthOverride=40}
   slot {class=OverlaySlot, properties {horizontalAlignment=HAlign_Right, padding=<ltrb 0 -16 -16 0>}}
  }
  Overlay_0 {widgetClassPath=&Overlay, parent=SB_Check, widget=Overlay_0, slot {class=SizeBoxSlot}}
  Image_55 {
   widgetClassPath=&Image
   parent=Overlay_0
   widget=Image_55
   properties {
    visibility=SelfHitTestInvisible
    brush {
     bIsDynamicallyLoaded=false
     drawAs=Image
     imageType=NoImage
     mirroring=NoMirror
     resourceName=None
     resourceObject=/Game/UI/Foundation/Textures/Icons/Checks/T-Icon-Check-128
     tiling=NoTile
     imageSize=<34 34>
     margin=<ltrb 0 0 0 0>
     outlineSettings {
      bUseBrushTransparency=false
      roundingType=HalfHeightRadius
      width=0
      color {colorUseRule=UseColor_Specified, specifiedColor=<rgba 0 0 0 0>}
      cornerRadii=<0 0 0 0>
     }
     tintColor {colorUseRule=UseColor_Specified, specifiedColor=<rgba 1 1 1 1>}
     uVRegion {bIsValid=false, max=<0 0>, min=<0 0>}
    }
    colorAndOpacity=<rgba 0.0194 0.0194 0.0666 0.8>
    renderTransform {angle=4, scale=<1 1>, shear=<0 0>, translation=<2 3>}
    renderTransformPivot=<0.3439 1>
   }
   slot {class=OverlaySlot, properties {horizontalAlignment=HAlign_Center, verticalAlignment=VAlign_Center}}
  }
  Image_85 {
   widgetClassPath=&Image
   parent=Overlay_0
   widget=Image_85
   properties {
    visibility=SelfHitTestInvisible
    brush {
     bIsDynamicallyLoaded=false
     drawAs=Image
     imageType=NoImage
     mirroring=NoMirror
     resourceName=None
     resourceObject=&T-Icon-Check-128
     tiling=NoTile
     imageSize=<34 34>
     margin=<ltrb 0 0 0 0>
     outlineSettings {
      bUseBrushTransparency=false
      roundingType=HalfHeightRadius
      width=0
      color {colorUseRule=UseColor_Specified, specifiedColor=<rgba 0 0 0 0>}
      cornerRadii=<0 0 0 0>
     }
     tintColor {colorUseRule=UseColor_Specified, specifiedColor=<rgba 1 1 1 1>}
     uVRegion {bIsValid=false, max=<0 0>, min=<0 0>}
    }
    colorAndOpacity=<rgba 0 1 0.1119 1>
   }
   slot {class=OverlaySlot, properties {horizontalAlignment=HAlign_Center, verticalAlignment=VAlign_Center}}
  }
  Button_Internal {
   widgetClassPath=/FortUILibrary/UIKit/Buttons/WBP_UIKit_Button_Quiet
   bInherited=true
   bIsVariable=true
   parent=Overlay_1
   widget=Button_Internal
   properties {text=Stay As A Team, triggeringInputAction {dataTable=&InputActionDataTable, rowName=GenericRightShoulder}}
   slot {class=OverlaySlot, properties {horizontalAlignment=HAlign_Fill, verticalAlignment=VAlign_Fill}}
  }
 ]
}
