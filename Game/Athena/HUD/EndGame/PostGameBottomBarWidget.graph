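# Post-game bottom bar HUD element, derived from FortHUDElementWidget, shown in the Athena end-game UI.
# A root border holds a dark, full-stretch background image layered with WBP_CaptureForPostBufferUpdate, plus an empty SafeZone with top padding disabled.
# On Construct, it collapses the whole border when IsMobileGame is true and shows it otherwise.
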WidgetBlueprint /Game/Athena/HUD/EndGame/PostGameBottomBarWidget
graphs [EventGraph {nodes [
   Event Construct()
   > SetVisibility(
    self=Get Border_BottomBorder() -> (Border_BottomBorder:Border Object Reference)
    InVisibility=Select(Option 0=Visible, Option 1=Collapsed, Index=IsMobileGame())
   )
  ]}]
variables []
widgets {
 info {inheritedWidgetCount=1, parentClass=&FortHUDElementWidget, rootWidgetClass=&Border, widgetCount=7}
 widgets [
  Border_BottomBorder {
   widgetClassPath=&Border
   bIsVariable=true
   widget=Border_BottomBorder
   properties {
    background {
     bIsDynamicallyLoaded=false
     drawAs=NoDrawType
     imageType=NoImage
     mirroring=NoMirror
     resourceName=None
     resourceObject=null
     tiling=NoTile
     imageSize=<32 32>
     margin=<ltrb 0 0 0 0>
     outlineSettings {
      bUseBrushTransparency=false
      roundingType=HalfHeightRadius
      width=0
      color {colorUseRule=UseColor_Specified, specifiedColor=<rgba 0 0 0 0>}
      cornerRadii=<0 0 0 0>
     }
     tintColor {colorUseRule=UseColor_Specified, specifiedColor=<rgba 1 1 1 1>}
     uVRegion {bIsValid=false, max=<0 0>, min=<0 0>}
    }
    padding=<ltrb 0 0 0 0>
   }
  }
  Overlay_0 {
   widgetClassPath=&Overlay
   parent=Border_BottomBorder
   widget=Overlay_0
   slot {class=BorderSlot, properties {padding=<ltrb 0 0 0 0>}}
  }
  Overlay_bgbar {
   widgetClassPath=&Overlay
   parent=Overlay_0
   widget=Overlay_bgbar
   slot {class=OverlaySlot, properties {horizontalAlignment=HAlign_Fill, verticalAlignment=VAlign_Fill}}
  }
  Image_topbar {
   widgetClassPath=&Image
   bIsVariable=true
   parent=Overlay_bgbar
   widget=Image_topbar
   properties {
    brush {
     bIsDynamicallyLoaded=false
     drawAs=Image
     imageType=NoImage
     mirroring=NoMirror
     resourceName=None
     resourceObject=null
     tiling=NoTile
     imageSize=<32 45>
     margin=<ltrb 0 0 0 0>
     outlineSettings {
      bUseBrushTransparency=false
      roundingType=HalfHeightRadius
      width=0
      color {colorUseRule=UseColor_Specified, specifiedColor=<rgba 0 0 0 0>}
      cornerRadii=<0 0 0 0>
     }
     tintColor {colorUseRule=UseColor_Specified, specifiedColor=<rgba 1 1 1 1>}
     uVRegion {bIsValid=false, max=<0 0>, min=<0 0>}
    }
    colorAndOpacity=<rgba 0.0033 0.004 0.0176 1>
   }
   slot {class=OverlaySlot, properties {horizontalAlignment=HAlign_Fill, verticalAlignment=VAlign_Fill}}
  }
  WBP_CaptureForPostBufferUpdate {
   widgetClassPath=&WBP_CaptureForPostBufferUpdate
   bIsVariable=true
   parent=Overlay_bgbar
   slot {class=OverlaySlot, properties {horizontalAlignment=HAlign_Fill, verticalAlignment=VAlign_Fill}}
  }
  SafeZone_1 {
   widgetClassPath=&SafeZone
   bIsVariable=true
   parent=Overlay_0
   widget=SafeZone_1
   properties {padTop=false}
   slot {class=OverlaySlot, properties {horizontalAlignment=HAlign_Fill, verticalAlignment=VAlign_Fill}}
  }
  TutorialHighlight {widgetClassPath=&TutorialHighlight}
 ]
}
