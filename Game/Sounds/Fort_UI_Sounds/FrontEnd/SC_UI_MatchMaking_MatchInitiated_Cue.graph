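# UI sound cue that, by its name, plays when matchmaking is initiated from the front end.
# A mixer layers a dominant start-game sound, a slightly delayed OK sound and a quieter click with its own quick fade-out.
# An outer enveloper fades the whole mix to silence in just over a second.
# It uses the tray notification sound class and overrides concurrency to one instance, stopping the oldest.
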SoundCue /Game/Sounds/Fort_UI_Sounds/FrontEnd/SC_UI_MatchMaking_MatchInitiated_Cue
root=Enveloper_0
nodes [
 {id=Delay_2, children [WavePlayer_4], properties {delayMax=0.02, delayMin=0.02}}
 {
  id=Enveloper_0
  children [Mixer_0]
  properties {
   pitchCurve {editorCurveData {keys [{time=0}]}}
   volumeCurve {editorCurveData {keys [<at 0 1>, {arriveTangent=-0.8, interpMode=RCIM_Cubic, leaveTangent=0.0461, tangentMode=RCTM_Break, time=0.7952, value=0.9158}, <at 1.25 0>]}}
  }
 }
 {
  id=Enveloper_6
  children [WavePlayer_27]
  properties {pitchCurve {editorCurveData {keys [{time=0}]}}, volumeCurve {editorCurveData {keys [<at 0 1>,<at 0.5 0>]}}}
 }
 {id=Mixer_0, children [WavePlayer_0,Delay_2,Enveloper_6], properties {inputVolume [1.5,0.7,0.2]}}
 {id=WavePlayer_0, properties {soundWaveAssetPtr=/Game/Sounds/Fort_UI_Sounds/FrontEnd/New_FrontEnd/Fortnite_UI_MainMenu_StartGame_01}}
 {id=WavePlayer_27, properties {soundWaveAssetPtr=&/fort_new_frontend_ui_click_02}}
 {id=WavePlayer_4, properties {soundWaveAssetPtr=&/fort_ui_ok_02}}
]
properties {
 bOverrideConcurrency=true
 maxDistance=2097152
 soundClassObject=/Game/Sounds/FortSoundClassesAndModes/Fort_Application_TrayNotification
 volumeMultiplier=0.3991
 concurrencyOverrides {maxCount=1, resolutionRule=StopOldest}
}
